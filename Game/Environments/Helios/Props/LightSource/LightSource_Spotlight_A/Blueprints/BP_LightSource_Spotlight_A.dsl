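; Destructible spotlight prop with light shaft and flare meshes, configured mostly in its construction script.
; It hashes world X/Y into a deterministic per-placement rotation for the shaft material, desynchronizing instances without random streams.
; Dynamic materials receive flare rotation settings and the spotlight color, which also tints the light and housing emissive; Switch disables light shadows.
; On death, it calls the parent effects and fades out the mechanism loop audio.
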(event EventBeginPlay)

(event EventOnDeathPlayEffects
  (params (Damage :type Float%20%28single-precision%29) (DamageTags :type Gameplay%20Tag%20Container%20Structure%20%28by%20ref%29) (Momentum :type Vector :default "0, 0, 0") (HitInfo :type Hit%20Result%20Structure%20%28by%20ref%29) (InstigatedBy :type Fort%20Pawn%20Object%20Reference) (DamageCauser :type Actor%20Object%20Reference) (EffectContext :type Gameplay%20Effect%20Context%20Handle%20Structure))
  (|CallParentFunction Damage DamageTags Momentum HitInfo InstigatedBy DamageCauser EffectContext)
  (bind _spot_light_mechanism_loop_cue (Variables|Audio|GetSpot_Light_Mechanism_Loop_Cue)
    (:then
      (Audio|Components|Audio|FadeOut _spot_light_mechanism_loop_cue 0.35 0.0 "Linear" :!class /Script/Engine.AudioComponent))
    (:else)))

(fn ConstructionScript
  (bind _spotlightcolor (Variables|SpotlightFXSettings|GetSpotlightColor))
  (bind _sm_neo_tilted_light_shaft_flares (Variables|Default|Getsm_neo_tilted_light_shaft_Flares))
  (|CallParentFunction)
  (Utilities|FlowControl|Sequence
    (:then_0
      (Variables|Default|SetRotation (* (Math|Float|Absolute(Float) (Math|Float|Fraction (Math|Float|SafeDivide (.x (Transformation|GetWorldLocation (Variables|Debug|GetStaticMeshComponent))) (.y (Transformation|GetWorldLocation (Variables|Debug|GetStaticMeshComponent)))))) 333.330017)))
    (:then_1
      (Rendering|Components|Light|SetLightColor (Variables|Default|GetPointLight) (Variables|SpotlightFXSettings|GetSpotlightColor) true)
      (if (Utilities|String|EqualExactly(String) (Game|GetPlatformName) "Switch")
        (Rendering|Components|Light|SetCastShadows (Variables|Default|GetPointLight) false)))
    (:then_2
      (bind _returnvalue (Rendering|Material|CreateDynamicMaterialInstance (Variables|Default|GetS_NeoTilted_Spotlight_01) 0 0 "spolightMeshMID" :!class /Script/Engine.PrimitiveComponent))
      (bind _returnvalue_2 (Rendering|Material|CreateDynamicMaterialInstance (Variables|Default|Getsm_neo_tilted_light_shaft) 0 0 "spotlightShaftMID" :!class /Script/Engine.PrimitiveComponent))
      (bind _returnvalue_3 (Rendering|Material|CreateDynamicMaterialInstance _sm_neo_tilted_light_shaft_flares 0 0 "spotlightFlaresMID" :!class /Script/Engine.PrimitiveComponent))
      (bind _returnvalue_4 (Rendering|Material|CreateDynamicMaterialInstance _sm_neo_tilted_light_shaft_flares 1 0 "spotlightFlaresMIDSecond" :!class /Script/Engine.PrimitiveComponent))
      (Rendering|Material|SetScalarParameterValue _returnvalue_2 "rotation" (Variables|Default|GetRotation) :!class /Script/Engine.MaterialInstanceDynamic)
      (Rendering|Material|SetScalarParameterValue _returnvalue_3 "Rotation Max" (Variables|SpotlightFXSettings|GetRotation%20Max) :!class /Script/Engine.MaterialInstanceDynamic)
      (Rendering|Material|SetScalarParameterValue _returnvalue_3 "Rotation Speed" (Variables|SpotlightFXSettings|GetRotation%20Speed) :!class /Script/Engine.MaterialInstanceDynamic)
      (Rendering|Material|SetScalarParameterValue _returnvalue_4 "Rotation Speed2" 0.07 :!class /Script/Engine.MaterialInstanceDynamic)
      (Rendering|Material|SetVectorParameterValue _returnvalue_3 "Color" _spotlightcolor :!class /Script/Engine.MaterialInstanceDynamic)
      (Rendering|Material|SetVectorParameterValue _returnvalue "Emissive Color Multiplier" (Math|Color|LinearColor*Float _spotlightcolor (Variables|SpotlightFXSettings|GetspotlightMeshEmissiveMult)) :!class /Script/Engine.MaterialInstanceDynamic))))
